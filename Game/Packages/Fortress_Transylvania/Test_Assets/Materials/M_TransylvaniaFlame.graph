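# M_TransylvaniaFlame is an unlit translucent material for an animated mesh flame in the Transylvania fortress test assets.
# Inverted fresnel, vertical UV gradient and panning noise form a mask that indexes a fire ramp and shapes opacity.
# Tinted ramp plus fresnel glow feeds emissive, with optional hue shift and saturation control.
# Layered sine waves, phase-shifted by component pivot, drive vertex wobble, sway and bobbing, gradient-damped on Low and Medium quality.
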Material /Game/Packages/Fortress_Transylvania/Test_Assets/Materials/M_TransylvaniaFlame
graph {
 properties {
  bUsedWithInstancedStaticMeshes=true
  bUsedWithMeshDeformer=true
  bUsedWithNiagaraMeshParticles=true
  bUsedWithSkeletalMesh=true
  blendMode=BLEND_Translucent
  pixelDepthOffsetMode=PDOM_Legacy
  shadingModel=MSM_Unlit
  usedShadingModels=Unlit
 }
 nodes [
  ComponentMask(None=TextureCoordinate() {vTiling=6}) {g=true}
  n1 = Constant() {r=2.5}
  n2 = Multiply(
   A=ComponentMask(None=TextureCoordinate()) {g=true}
   B=Multiply(
    A=VectorParameter() {parameterName=FresnelColor, defaultValue=<rgba 1 0.6981 0.4503 1>}.RGB
    B=MaterialFunctionCall(Exponent=@n1) {materialFunction=&MF_Fresnel_FX}.Fresnel Inverted
   )
  )
  n3 = OneMinus(None=TextureCoordinate())
  n4 = ScalarParameter() {defaultValue=1.5, parameterName=FresnelExp}
  n5 = ComponentMask(None=TextureCoordinate()) {g=true}
  n6 = Add(A=CameraVectorWS(), B=Constant3Vector() {constant=<rgba 0 0 -0.5 0>})
  n7 = Multiply(A=MaterialFunctionCall(Exponent=Multiply(A=@n4, B=@n5), Camera Vector=Normalize(VectorInput=@n6)) {materialFunction=&MF_Fresnel_FX}.Fresnel Inverted)
  n8 = Clamp(None=Add(A=OneMinus(None=@n5)) {constB=-0.75})
  n9 = Multiply(A=Multiply(A=@n8, B=@n8)) {constB=4}
  n10 = TextureSample(UVs=Panner(Coordinate=TextureCoordinate() {uTiling=2, vTiling=0.6}) {bFractionalPart=true, speedX=0.17, speedY=2.5}) {texture=&T_NoisePack_01}
  n11 = Clamp(None=Add(A=Add(A=@n7, B=@n9) {constB=0}, B=Multiply(A=@n10.R) {constB=0.25}))
  n12 = MaterialFunctionCall(X=@n11, Y=Constant() {r=0.5}) {materialFunction=&MakeFloat2}
  n13 = Multiply(
   A=LinearInterpolate(
    A=VectorParameter() {parameterName=ColorOutside, defaultValue=<rgba 1 1 1 1>}.RGB
    B=VectorParameter() {parameterName=ColorInside, defaultValue=<rgba 0.67 0.3384 0.1822 1>}.RGB
    Alpha=ComponentMask(None=@n3) {g=true}
   )
   B=Multiply(A=TextureSample(UVs=@n12.Result) {texture=/Game/Packages/Fortress_Transylvania/Test_Assets/Textures/T_FireRamp}.RGB) {constB=10}
  )
  n14 = StaticSwitchParameter(
   True=MaterialFunctionCall(Hue Shift Percentage=ScalarParameter() {parameterName=HueShiftPercent}, Texture=@n13) {materialFunction=&HueShift}.Result
   False=@n13
  ) {parameterName=UseHueShift}
  n15 = ScalarParameter() {parameterName=Saturation}
  MaterialFunctionCall(Emissive=Desaturation(None=Add(A=@n2, B=@n14), Fraction=Multiply(A=@n15) {constB=-1})) {materialFunction=&MF_FX_EmissiveAdjustments}
  n16 = MaterialFunctionCall(Float3=MaterialFunctionCall() {materialFunction=&ComponentPivotLocation}.Result) {materialFunction=&BreakOutFloat3Components}
  n17 = Add(A=Add(A=@n16.R, B=@n16.R), B=@n16.B)
  n18 = Add(A=Time(), B=Frac(None=Multiply(A=@n17) {constB=0.01}))
  n19 = Multiply(A=Frac(None=@n18)) {constB=7.1}
  n20 = OneMinus(None=ComponentMask(None=TextureCoordinate()) {g=true})
  n21 = Multiply(A=Multiply(A=@n20, B=@n20)) {constB=-2}
  n22 = Multiply(A=Sine(None=Add(A=@n19, B=@n21)), B=ScalarParameter() {defaultValue=1.2, parameterName=SmallMoveStrength}) {constB=1.2, desc=Amplitude}
  n23 = Add(A=@n18, B=ComponentMask(None=TextureCoordinate()) {g=true})
  n24 = ScalarParameter() {defaultValue=1.1, parameterName=MediumMoveStrength}
  n25 = TextureCoordinate()
  n26 = Add(A=Multiply(A=@n18) {constB=1.31}, B=Multiply(A=ComponentMask(None=@n25) {g=true}) {constB=0.63})
  n27 = Add(
   A=Multiply(A=Sine(None=@n23) {period=1.31}, B=@n24) {constB=1.1, desc=Amplitude}
   B=Multiply(A=Sine(None=@n26) {period=1.31}, B=@n24) {constB=1.1, desc=Amplitude}
  )
  n28 = Multiply(A=VertexNormalWS(), B=Max(A=Add(A=@n22, B=@n27)) {constB=-2.5})
  n29 = AppendVector(A=ComponentMask(None=WorldPosition().XYZ) {b=true}, B=Constant2Vector())
  n30 = Add(A=Multiply(A=Frac(None=@n18)) {constB=4}, B=Frac(None=Multiply(A=@n29) {constB=-0.02}))
  n31 = ScalarParameter() {defaultValue=2, parameterName=LargeShapeStrength}
  n32 = Multiply(A=Sine(None=@n30), B=@n31) {desc=Amplitude}
  n33 = Multiply(A=Cosine(None=@n30), B=@n31)
  n34 = OneMinus(None=ComponentMask(None=TextureCoordinate()) {g=true})
  n35 = Multiply(
   A=MaterialFunctionCall(X=ComponentMask(None=@n32) {r=true}, Y=ComponentMask(None=@n33) {r=true}, Z=Constant()) {materialFunction=&MakeFloat3}.Result
   B=Multiply(A=@n34, B=@n34)
  )
  n36 = Add(A=ComponentMask(None=TextureCoordinate()) {g=true}) {constB=0.5}
  n37 = Sine(None=@n18) {period=0.5}
  n38 = Sine(None=@n18) {period=0.73}
  n39 = Multiply(A=Add(A=Multiply(A=@n37) {constB=0.79}, B=Multiply(A=@n38) {constB=0.5}), B=ScalarParameter() {defaultValue=3, parameterName=UpDownStrength}) {constB=3}
  n40 = ComponentMask(None=OneMinus(None=TextureCoordinate())) {g=true}
  n41 = Add(A=Multiply(A=Add(A=@n28, B=@n35) {constB=0}, B=Clamp(None=@n36)), B=AppendVector(A=Constant2Vector(), B=Multiply(A=@n39, B=@n40)))
  n42 = MaterialFunctionCall() {materialFunction=&LinearGradient}
  n43 = Multiply(A=@n42.VGradient, B=@n42.VGradient)
  n44 = Multiply(A=OneMinus(None=@n42.UGradient), B=@n42.VGradient)
  n45 = Multiply(A=@n41, B=Power(Base=Multiply(A=@n43, B=@n44)) {constExponent=0.5})
  QualitySwitch(Default=@n41, Low=@n45, High=@n41, Medium=@n45, Epic=@n41)
  n46 = OneMinus(None=@n11)
  n47 = Clamp(None=Add(A=Power(Base=@n11) {constExponent=8}, B=Power(Base=@n46) {constExponent=4}) {constB=0})
  StaticSwitchParameter(True=Multiply(A=@n47, B=ScalarParameter() {defaultValue=1, parameterName=OpacityMult}), False=@n47) {parameterName=UseOpacity}
 ]
}
